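# Animation Blueprint for the clown mask hat cosmetic, adding physics wobble to the hat and hair and continuously spinning both eyes.
# Three chained Anim Dynamics nodes simulate the hat and left and right hair bones in actor space with angular limits, feeding two Modify Bone nodes that replace eye rotations.
# The update event advances SpinX by SpinPerSecond and delta time, wraps it at 720, and writes the shared eyeSpin roll.
# An unconnected RigidBody node is unused.
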AnimBlueprint /BRCosmetics/Accessories/Hats/M_MED_ClownMask_01/Meshes/M_MED_ClownMask_01_AnimBP
target_skeleton=&/M_MED_ClownMask_01_Skeleton
graphs [
 AnimGraph {nodes [
   RigidBody() {alphaBoolBlend {blendOption=Linear}}
   AnimGraphNode_SubInput_0 = Input Pose()
   n1 = Anim Dynamics - Bone: dyn_hat(ComponentPose=Local To Component(LocalPose=@AnimGraphNode_SubInput_0), Alpha=0.8, ExternalForce="0,0,0") {
    angularDampingOverride=0.9
    bGravityOverrideInSimSpace=true
    bOverrideAngularDamping=true
    bUsePlanarLimit=false
    numSolverIterationsPostUpdate=2
    numSolverIterationsPreUpdate=8
    simulationSpace=Actor
    alphaBoolBlend {blendOption=Linear}
    boundBone {boneName=dyn_hat}
    physicsBodyDefinitions [{
      sphereCollisionRadius=0
      boundBone {boneName=dyn_hat}
      boxExtents=<10 10 5>
      constraintSetup {angularTargetAxis=AxisZ, twistAxis=AxisZ, angularLimitsMax=<15 15 0>, angularLimitsMin=<-15 -15 0>, angularTarget=<0 0 0>}
      localJointOffset=<0 0 -0.1>
     }]
   }
   n2 = Anim Dynamics - Bone: dyn_hair_r(ComponentPose=@n1, Alpha=0.7, ExternalForce="0,0,0") {
    angularDampingOverride=0.8
    bGravityOverrideInSimSpace=true
    bOverrideAngularDamping=true
    bUsePlanarLimit=false
    numSolverIterationsPostUpdate=2
    numSolverIterationsPreUpdate=8
    simulationSpace=Actor
    alphaBoolBlend {blendOption=Linear}
    boundBone {boneName=dyn_hair_r}
    physicsBodyDefinitions [{
      sphereCollisionRadius=0
      boundBone {boneName=dyn_hair_r}
      boxExtents=<5 5 2.5>
      constraintSetup {angularTargetAxis=AxisZ, twistAxis=AxisZ, angularLimitsMax=<15 15 0>, angularLimitsMin=<-15 -15 0>, angularTarget=<0 0 0>}
      localJointOffset=<0 0 -0.1>
     }]
   }
   n3 = Anim Dynamics - Bone: dyn_hair_l(ComponentPose=@n2, Alpha=0.75, ExternalForce="0,0,0") {
    angularDampingOverride=0.85
    bGravityOverrideInSimSpace=true
    bOverrideAngularDamping=true
    bUsePlanarLimit=false
    numSolverIterationsPostUpdate=2
    numSolverIterationsPreUpdate=8
    simulationSpace=Actor
    alphaBoolBlend {blendOption=Linear}
    boundBone {boneName=dyn_hair_l}
    physicsBodyDefinitions [{
      sphereCollisionRadius=0
      boundBone {boneName=dyn_hair_l}
      boxExtents=<5 5 2.5>
      constraintSetup {angularTargetAxis=AxisZ, twistAxis=AxisZ, angularLimitsMax=<15 15 0>, angularLimitsMin=<-15 -15 0>, angularTarget=<0 0 0>}
      localJointOffset=<-0.01 0 -0.1>
     }]
   }
   n4 = Get eyeSpin() -> (eyeSpin:Rotator)
   n5 = "Transform (Modify) Bone - Bone: dyn_eye_l"(ComponentPose=@n3, Alpha=1, Translation="0,0,0", Rotation=@n4, Scale="1,1,1") {rotationMode=BMM_Replace, rotationSpace=BCS_BoneSpace, alphaBoolBlend {blendOption=Linear}, boneToModify {boneName=dyn_eye_l}}
   n6 = "Transform (Modify) Bone - Bone: dyn_eye_r"(ComponentPose=@n5, Alpha=1, Translation="0,0,0", Rotation=@n4, Scale="1,1,1") {rotationMode=BMM_Replace, rotationSpace=BCS_BoneSpace, alphaBoolBlend {blendOption=Linear}, boneToModify {boneName=dyn_eye_r}}
   Output Pose(Result=Component To Local(ComponentPose=@n6)) {layerGroup=None}
  ]}
 EventGraph {nodes [
   n1 = Event BlueprintUpdateAnimation()
   > n2 = Branch(Condition="float >= float"(A=Get SpinX() -> (SpinX:double), B=720) {is=title K2Node_CallFunction})
   > n3 = Set SpinX(SpinX:double=float - float(A=Get SpinX(), B=720) {is=title K2Node_CallFunction}) -> (Output_Get:double)
   n4 = float + float(A=Get SpinX(), B=float * float(A=@n1.DeltaTimeX, B=Get SpinPerSecond() -> (SpinPerSecond:double))) {is=title K2Node_CommutativeAssociativeBinaryOperator}
   Set eyeSpin(execute=[@n5.then, @n3.then], eyeSpin:Rotator=MakeRotator(Roll=@n4)) -> (Output_Get:Rotator)
   n5 = Set SpinX(execute=@n2.else, SpinX=@n4)
  ]}
]
variables [
 eyeSpin {type=Rotator}
 SpinPerSecond {type=number, default=150}
 SpinX {type=number}
]
